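; Feat quest ability that credits a player for eliminating a carried enemy through fall damage after a throw.
; On event activation it finds the attached player pawn, ending if none is valid or it is a teammate.
; It then waits for that pawn's falling-damage event, sending a quest stat event if the pawn is dead.
; A landed delegate ends the ability after a short delay, leaving time for the damage event.
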(event Ability|EventActivateAbilityFromEvent
  (params (EventData :type Gameplay%20Event%20Data%20Structure%20%28by%20ref%29))
  (bind _carriedpawn (Variables|Default|GetCarriedPawn))
  (bind _carriedpawn_2 (Variables|Default|GetCarriedPawn))
  (|CallParentFunction EventData)
  ;[1 1 1] Find pawn being carried
  (for _array_element (Actor|GetAttachedActors (Variables|GAVars|Data|GetOwningPlayerPawn) true false)
    (Utilities|IsValid _array_element
      (:Is%20Valid
        (bind _asfort_player_pawn_athena (Utilities|Casting|CastToFortPlayerPawnAthena _array_element))
        (Variables|Default|SetCarriedPawn _asfort_player_pawn_athena))))
  ;[1 1 1] Check carried pawn's team
  (Utilities|IsValid _carriedpawn
    (:Is%20Valid
      (if (Team|OnSameTeam _carriedpawn (Variables|GAVars|Data|GetOwningPlayerPawn))
        (Ability|EndAbility)
        (else
          ;[1 1 1] If enemy wait for fall damage
          (Ability|Tasks|WaitGameplayEvent "(TagName=\"Event.Damage.Received.Falling\")" (Variables|Default|GetCarriedPawn)
            (:then
              ;[1 1 1] When pawn lands end ability
              (EventDispatchers|BindEventtoOnPawnLanded (Variables|Default|GetCarriedPawn) (delegate Custom|Pawn%20Landed)))
            (:EventReceived
              ;[1 1 1] If fall damage recieved grant quest progress
              (if (FortPawn|IsDead _carriedpawn_2)
                (Quest|SendComplexCustomStatEvent (Variables|GAVars|Data|GetOwningPlayerQuestManager) _carriedpawn_2 "(GameplayTags=)" (Variables|Default|GetYeetEliminationTag) 1)
                (else
                  _))
              (Ability|EndAbility))))))
    (:Is%20Not%20Valid
      (Ability|EndAbility))))

(event Custom|Pawn%20Landed
  (params (Hit :type Hit%20Result%20Structure%20%28by%20ref%29))
  ;[1 1 1] When pawn lands end ability
  (Utilities|FlowControl|Delay 0.5)
  (EventDispatchers|UnbindEventfromOnPawnLanded (Variables|Default|GetCarriedPawn) (delegate Custom|Pawn%20Landed))
  (Ability|EndAbility))
